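# Physics asset for the pink Raider female medium backpack cosmetic, giving its dangling gun, straps and zipper physical bodies and joints.
# The pack body uses two angled capsules, while each dynamic bone has a single box.
# Gun and zipper joint to the pack, and two strap segments chain off the gun.
# Angular limits keep the gun nearly rigid while straps and zipper swing widely on one axis with minimal twist.
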PhysicsAsset /BRCosmetics/Accessories/FORT_Backpacks/Backpack_F_MED_Raider_Pink/Meshes/F_MED_Raider_Pink_Physics
bodies {
 dyn_gun [{shapeType=Box, extentX=6.4228, extentY=26.3099, extentZ=13.563, center=<3.6122 -0.8149 -3.0891>}]
 dyn_strap_01 [{shapeType=Box, extentX=5.8307, extentY=4.2251, extentZ=1.9575, center=<2.5288 -0.0853 0.0039>}]
 dyn_strap_02 [{shapeType=Box, extentX=3.5006, extentY=5.2883, extentZ=1.2205, center=<1.2379 0.0875 0.1411>}]
 dyn_zipper [{shapeType=Box, extentX=3.7679, extentY=3.1136, extentZ=1.0577, center=<1.4979 0.0291 -0.1224>}]
 pack [
  {shapeType=Capsule, length=21.314, radius=4.567, center=<1.2219 -3.4548 -1.5948>, rotation=<rot -72.1893 0 0>}
  {shapeType=Capsule, length=16.772, radius=3.6039, center=<2.2574 -3.3205 -6.2253>, rotation=<rot -71.2534 0 0>}
 ]
}
constraints [
 {bone1Name=dyn_gun, bone2Name=pack, swing1LimitDegrees=1, swing2LimitDegrees=7, twistLimitDegrees=5}
 {bone1Name=dyn_zipper, bone2Name=pack, swing1LimitDegrees=10, swing2LimitDegrees=120, twistLimitDegrees=1}
 {bone1Name=dyn_strap_01, bone2Name=dyn_gun, swing1LimitDegrees=25, swing2LimitDegrees=70, twistLimitDegrees=1}
 {bone1Name=dyn_strap_02, bone2Name=dyn_strap_01, swing1LimitDegrees=5, swing2LimitDegrees=90, twistLimitDegrees=1}
]
